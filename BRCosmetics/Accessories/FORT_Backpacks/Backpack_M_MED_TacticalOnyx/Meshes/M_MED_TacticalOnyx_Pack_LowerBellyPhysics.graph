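# Physics asset for the lower-belly part of the M_MED_TacticalOnyx backpack cosmetic, defining bodies for its dynamic ear and squash bones.
# Each bone has one capsule: a wide one on the squash base, small mirrored ones on the left and right ears, and a tiny one on the pack bone.
# Three constraints tie the ears and squash base directly to the pack in a star layout, with wide swing angles and near-zero twist angles.
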PhysicsAsset /BRCosmetics/Accessories/FORT_Backpacks/Backpack_M_MED_TacticalOnyx/Meshes/M_MED_TacticalOnyx_Pack_LowerBellyPhysics
bodies {
 dyn_Ear_l [{
   shapeType=Capsule
   length=2.9261
   radius=2.0695
   shapeName=dyn_Ear_l_capsule
   center=<2.1236 0.2223 -0.3277>
   rotation=<rot 0.3445 -92.499 -101.4393>
  }]
 dyn_Ear_r [{
   shapeType=Capsule
   length=2.9261
   radius=2.0695
   shapeName=dyn_Ear_l_capsule
   center=<1.8853 -0.3973 0.3118>
   rotation=<rot 0.3444 -92.4991 102.5341>
  }]
 dyn_Squash_Base [{
   shapeType=Capsule
   length=2.2833
   radius=11.0161
   shapeName=dyn_Squash_Base_capsule
   center=<-0.0149 -0.1382 -0.1637>
   rotation=<rot -4.1054 -88.2378 -89.6012>
  }]
 pack [{shapeType=Capsule, length=0.505, radius=0.505, shapeName=pack_capsule, rotation=<rot 0 -90 0>}]
}
constraints [
 {bone1Name=dyn_Squash_Base, bone2Name=pack, swing1LimitDegrees=60, swing2LimitDegrees=30, twistLimitDegrees=0}
 {bone1Name=dyn_Ear_l, bone2Name=pack, swing1LimitDegrees=70, swing2LimitDegrees=30, twistLimitDegrees=1}
 {bone1Name=dyn_Ear_r, bone2Name=pack, swing1LimitDegrees=70, swing2LimitDegrees=30, twistLimitDegrees=1}
]
