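# Physics asset for the TacticalOnyx medium male backpack cosmetic, providing bodies and joints for its dynamic bones.
# It has three capsule bodies: a tiny one on pack, a large one on dyn_Squash_Base and a small one on Dyn_Chain.
# The squash base has locked swing and twist relative to the pack.
# The chain hangs off the squash base with angular motion left unlocked, so it can swing.
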PhysicsAsset /BRCosmetics/Accessories/FORT_Backpacks/Backpack_M_MED_TacticalOnyx/Meshes/M_MED_TacticalOnyx_Pack_Physics
bodies {
 Dyn_Chain [{
   shapeType=Capsule
   length=1.2088
   radius=0.8564
   shapeName=Dyn_Chain_capsule
   center=<2.2792 -0.0888 -0.5004>
   rotation=<rot -14.8702 -90.259 -77.5394>
  }]
 dyn_Squash_Base [{
   shapeType=Capsule
   length=1.4304
   radius=10.9708
   shapeName=dyn_Squash_Base_capsule
   center=<0.4016 0.0432 0.2025>
   rotation=<rot 0.1896 -87.8837 93.4131>
  }]
 pack [{shapeType=Capsule, length=0.505, radius=0.505, shapeName=pack_capsule, rotation=<rot 0 -90 0>}]
}
constraints [
 {
  bone1Name=dyn_Squash_Base
  bone2Name=pack
  swing1LimitDegrees=40
  swing1Motion=Locked
  swing2LimitDegrees=40
  swing2Motion=Locked
  twistLimitDegrees=0
  twistMotion=Locked
 }
 {bone1Name=Dyn_Chain, bone2Name=dyn_Squash_Base, swing1LimitDegrees=110, swing2LimitDegrees=110, twistLimitDegrees=0}
]
