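# Anim Blueprint for the TacticalOnyx backpack cosmetic, adding secondary motion and skydive adjustments to its skeleton.
# Spring-driven AnimDynamics on arm, thigh and lightning-bolt bones feeds into and surrounds two RigidBody simulations using whole-pack and lower-belly physics assets.
# Additive bone transforms then reposition pack and thighs, with blended toggles for skydiving, floating and parachute states.
# The EventGraph only clears the parachute flag, using DoOnce and a delay.
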AnimBlueprint /BRCosmetics/Accessories/FORT_Backpacks/Backpack_M_MED_TacticalOnyx/Meshes/M_MED_TacticalOnyx_Pack_AnimBP
target_skeleton=&/M_MED_TacticalOnyx_Pack_Skeleton
graphs [
 AnimGraph {nodes [
   n1 = Anim Dynamics - Bone: dyn_Arm_L(ComponentPose=Local To Component(LocalPose=Input Pose()), Alpha=1, ExternalForce="0,0,0") {
    angularDampingOverride=0.9999
    angularSpringConstant=8000
    bAngularSpring=true
    bOverrideAngularDamping=true
    bUseGravityOverride=true
    gravityScale=0
    alphaBoolBlend {blendOption=Linear}
    boundBone {boneName=dyn_Arm_L}
    componentAppliedLinearAccClamp=<5000 5000 8000>
    componentLinearAccScale=<0.5 0.5 0.4>
    componentLinearVelScale=<0.5 0.5 0.5>
    gravityOverride {z=-650}
    physicsBodyDefinitions [{
      boundBone {boneName=dyn_Arm_L}
      boxExtents=<15 10 10>
      constraintSetup {angularLimitsMax=<0 25 25>, angularLimitsMin=<0 -20 -20>, angularTarget=<0 0 0>}
      localJointOffset=<0.8 0.6 0>
     }]
    relativeSpaceBone {boneName=pack}
   }
   n2 = Anim Dynamics - Bone: dyn_Arm_r(ComponentPose=@n1, Alpha=1, ExternalForce="0,0,0") {
    angularDampingOverride=0.9999
    angularSpringConstant=8000
    bAngularSpring=true
    bOverrideAngularDamping=true
    bUseGravityOverride=true
    gravityScale=0
    alphaBoolBlend {blendOption=Linear}
    boundBone {boneName=dyn_Arm_r}
    componentAppliedLinearAccClamp=<5000 5000 8000>
    componentLinearAccScale=<0.5 0.5 0.4>
    componentLinearVelScale=<0.5 0.5 0.5>
    gravityOverride {z=-650}
    physicsBodyDefinitions [{
      boundBone {boneName=dyn_Arm_r}
      boxExtents=<13 7 7>
      constraintSetup {angularLimitsMax=<0 25 25>, angularLimitsMin=<0 -20 -20>, angularTarget=<0 0 0>}
      localJointOffset=<0.8 0.6 0>
     }]
    relativeSpaceBone {boneName=pack}
   }
   n3 = Anim Dynamics - Bone: dyn_Thigh_L(ComponentPose=@n2, Alpha=1, ExternalForce="0,0,0") {
    angularDampingOverride=0.9999
    angularSpringConstant=8000
    bAngularSpring=true
    bOverrideAngularDamping=true
    bUseGravityOverride=true
    gravityScale=0
    alphaBoolBlend {blendOption=Linear}
    boundBone {boneName=dyn_Thigh_L}
    componentAppliedLinearAccClamp=<5000 5000 8000>
    componentLinearAccScale=<0.5 0.5 0.4>
    componentLinearVelScale=<0.5 0.5 0.5>
    gravityOverride {z=-650}
    physicsBodyDefinitions [{
      boundBone {boneName=dyn_Thigh_L}
      boxExtents=<13 10 7>
      constraintSetup {angularLimitsMax=<0 30 30>, angularLimitsMin=<0 -30 -30>, angularTarget=<0 0 0>}
      localJointOffset=<0.8 0.6 0>
     }]
    relativeSpaceBone {boneName=pack}
   }
   n4 = Anim Dynamics - Bone: dyn_Thigh_r(ComponentPose=@n3, Alpha=1, ExternalForce="0,0,0") {
    angularDampingOverride=0.9999
    angularSpringConstant=8000
    bAngularSpring=true
    bOverrideAngularDamping=true
    bUseGravityOverride=true
    gravityScale=0
    alphaBoolBlend {blendOption=Linear}
    boundBone {boneName=dyn_Thigh_r}
    componentAppliedLinearAccClamp=<5000 5000 8000>
    componentLinearAccScale=<0.5 0.5 0.4>
    componentLinearVelScale=<0.5 0.5 0.5>
    gravityOverride {z=-650}
    physicsBodyDefinitions [{
      boundBone {boneName=dyn_Thigh_r}
      boxExtents=<13 10 7>
      constraintSetup {angularLimitsMax=<0 30 30>, angularLimitsMin=<0 -30 -30>, angularTarget=<0 0 0>}
      localJointOffset=<0.8 0.6 0>
     }]
    relativeSpaceBone {boneName=pack}
   }
   n5 = RigidBody(ComponentPose=@n4, Alpha=0.5) {
    bOverrideWorldGravity=true
    overridePhysicsAsset=&/M_MED_TacticalOnyx_Pack_Physics
    simulationSpace=BaseBoneSpace
    alphaBoolBlend {blendOption=Linear}
    baseBoneRef {boneName=root}
    componentAppliedLinearAccClamp=<850 850 850>
    componentLinearAccScale=<0.4 0.4 0.4>
    componentLinearVelScale=<0.4 0.4 0.4>
    overrideWorldGravity {z=-650}
   }
   n6 = RigidBody(ComponentPose=@n5) {
    bOverrideWorldGravity=true
    overridePhysicsAsset=&/M_MED_TacticalOnyx_Pack_LowerBellyPhysics
    simulationSpace=BaseBoneSpace
    alphaBoolBlend {blendOption=Linear}
    baseBoneRef {boneName=dyn_Squash_Base}
    componentLinearAccScale=<1.2 1.2 1.2>
    componentLinearVelScale=<1.2 1.2 1.2>
    overrideWorldGravity {z=-650}
   }
   n7 = Anim Dynamics - Bone: dyn_LightingBolts_Base_l(ComponentPose=@n6, Alpha=1, ExternalForce="0,0,0") {
    angularDampingOverride=0.9999
    angularSpringConstant=-8000
    bAngularSpring=true
    bOverrideAngularDamping=true
    bUseGravityOverride=true
    gravityScale=0
    alphaBoolBlend {blendOption=Linear}
    boundBone {boneName=dyn_LightingBolts_Base_l}
    componentAppliedLinearAccClamp=<5000 5000 8000>
    componentLinearAccScale=<0.5 0.5 0.4>
    componentLinearVelScale=<0.5 0.5 0.5>
    gravityOverride {z=-650}
    physicsBodyDefinitions [{
      boundBone {boneName=dyn_LightingBolts_Base_l}
      boxExtents=<8 6 6>
      constraintSetup {angularLimitsMax=<0 10 10>, angularLimitsMin=<0 -10 -10>, angularTarget=<0 0 0>}
      localJointOffset=<0.8 0.6 0>
     }]
    relativeSpaceBone {boneName=pack}
   }
   n8 = Anim Dynamics - Bone: dyn_LightingBolts_l(ComponentPose=@n7, Alpha=1, ExternalForce="0,0,0") {
    angularDampingOverride=0.9999
    angularSpringConstant=-8000
    bAngularSpring=true
    bOverrideAngularDamping=true
    bUseGravityOverride=true
    gravityScale=0
    alphaBoolBlend {blendOption=Linear}
    boundBone {boneName=dyn_LightingBolts_l}
    componentAppliedLinearAccClamp=<5000 5000 8000>
    componentLinearAccScale=<0.5 0.5 0.4>
    componentLinearVelScale=<0.5 0.5 0.5>
    gravityOverride {z=-650}
    physicsBodyDefinitions [{
      boundBone {boneName=dyn_LightingBolts_l}
      boxExtents=<7 6 6>
      constraintSetup {angularLimitsMax=<0 10 10>, angularLimitsMin=<0 -5 -5>, angularTarget=<0 0 0>}
      localJointOffset=<0.8 0.6 0>
     }]
    relativeSpaceBone {boneName=pack}
   }
   n9 = Anim Dynamics - Bone: dyn_LightingBolts_r(ComponentPose=@n8, Alpha=1, ExternalForce="0,0,0") {
    angularDampingOverride=0.9999
    angularSpringConstant=-8000
    bAngularSpring=true
    bOverrideAngularDamping=true
    bUseGravityOverride=true
    gravityScale=0
    alphaBoolBlend {blendOption=Linear}
    boundBone {boneName=dyn_LightingBolts_r}
    componentAppliedLinearAccClamp=<5000 5000 8000>
    componentLinearAccScale=<0.5 0.5 0.4>
    componentLinearVelScale=<0.5 0.5 0.5>
    gravityOverride {z=-650}
    physicsBodyDefinitions [{
      boundBone {boneName=dyn_LightingBolts_r}
      boxExtents=<8 6 6>
      constraintSetup {angularLimitsMax=<0 10 10>, angularLimitsMin=<0 -10 -10>, angularTarget=<0 0 0>}
      localJointOffset=<0.8 0.6 0>
     }]
    relativeSpaceBone {boneName=pack}
   }
   n10 = Anim Dynamics - Bone: dyn_LightingBolts_r(ComponentPose=@n9, Alpha=1, ExternalForce="0,0,0") {
    angularDampingOverride=0.9999
    angularSpringConstant=8000
    bAngularSpring=true
    bOverrideAngularDamping=true
    bUseGravityOverride=true
    gravityScale=0
    alphaBoolBlend {blendOption=Linear}
    boundBone {boneName=dyn_LightingBolts_r}
    componentAppliedLinearAccClamp=<5000 5000 8000>
    componentLinearAccScale=<0.5 0.5 0.4>
    componentLinearVelScale=<0.5 0.5 0.5>
    gravityOverride {z=650}
    physicsBodyDefinitions [{
      boundBone {boneName=dyn_LightingBolts_r}
      boxExtents=<7 6 6>
      constraintSetup {angularLimitsMax=<0 10 10>, angularLimitsMin=<0 -10 -10>, angularTarget=<0 0 0>}
      localJointOffset=<0.8 0.6 0>
     }]
    relativeSpaceBone {boneName=pack}
   }
   n11 = "Transform (Modify) Bone - Bone: pack"!(ComponentPose=@n10, bAlphaBoolEnabled:bool, Translation="0,-1.5,0", Rotation="0,0,0", Scale) -> (Pose) {
    tag=Fixing pack position
    alphaInputType=Bool
    rotationMode=BMM_Additive
    rotationSpace=BCS_BoneSpace
    translationMode=BMM_Additive
    translationSpace=BCS_BoneSpace
    alphaBoolBlend {blendOption=Linear, blendOutTime=0.4}
    boneToModify {boneName=pack}
    translation {y=-1.5}
   }
   n12 = Get bIsSkydiveFloating()
   n13 = "Transform (Modify) Bone - Bone: pack"(ComponentPose=@n11, bAlphaBoolEnabled=@n12, Translation="0,0,0", Rotation="0,0,15") {
    alphaInputType=Bool
    rotationMode=BMM_Additive
    rotationSpace=BCS_BoneSpace
    translationMode=BMM_Additive
    translationSpace=BCS_BoneSpace
    alphaBoolBlend {blendInTime=0.2, blendOption=Linear, blendOutTime=0.4}
    boneToModify {boneName=pack}
    rotation {roll=15}
   }
   n14 = "Transform (Modify) Bone - Bone: dyn_thigh_l"!(ComponentPose=@n13, bAlphaBoolEnabled:bool=@n12, Translation="0,0,0", Rotation="0,15,0", Scale="1,1,1") -> (Pose) {
    alphaInputType=Bool
    rotationMode=BMM_Additive
    rotationSpace=BCS_BoneSpace
    alphaBoolBlend {blendInTime=0.2, blendOption=Linear, blendOutTime=0.4}
    boneToModify {boneName=dyn_thigh_l}
    rotation {yaw=15}
   }
   n15 = "Transform (Modify) Bone - Bone: dyn_thigh_r"!(ComponentPose=@n14, bAlphaBoolEnabled:bool=@n12, Translation="0,0,0", Rotation="0,15,0", Scale="1,1,1") -> (Pose) {
    alphaInputType=Bool
    rotationMode=BMM_Additive
    rotationSpace=BCS_BoneSpace
    alphaBoolBlend {blendInTime=0.2, blendOption=Linear, blendOutTime=0.4}
    boneToModify {boneName=dyn_thigh_r}
    rotation {yaw=15}
   }
   n16 = "Transform (Modify) Bone - Bone: pack"(ComponentPose=@n15, bAlphaBoolEnabled=Get Parachute open() -> (Parachute open:bool), Translation="0,-0.5,0", Rotation="0,0,15") {
    alphaInputType=Bool
    rotationMode=BMM_Additive
    rotationSpace=BCS_BoneSpace
    translationMode=BMM_Additive
    translationSpace=BCS_BoneSpace
    alphaBoolBlend {blendInTime=0.3, blendOption=Linear, blendOutTime=0.3}
    boneToModify {boneName=pack}
    rotation {roll=15}
    translation {y=-0.5}
   }
   n17 = "Transform (Modify) Bone - Bone: pack"(ComponentPose=@n16, bAlphaBoolEnabled=Get bIsSkyDiving(), Translation="0,0,4", Rotation="0,0,0") {
    alphaInputType=Bool
    rotationMode=BMM_Additive
    rotationSpace=BCS_BoneSpace
    translationMode=BMM_Additive
    translationSpace=BCS_BoneSpace
    alphaBoolBlend {blendInTime=0.3, blendOption=Linear, blendOutTime=0.3}
    boneToModify {boneName=pack}
    translation {z=4}
   }
   Output Pose(Result=Component To Local(ComponentPose=@n17)) {layerGroup=None}
  ]}
 EventGraph {nodes [
   Event BlueprintUpdateAnimation()
   > n1 = Branch(Condition=Get Parachute open())
   > Do Once(Reset=@n1.else)
   > Set Parachute open(Parachute open:bool=false) -> (Output_Get:bool)
   > Delay(Duration=0.8)
   > Set Parachute open(Parachute open=false)
   TryGetPawnOwner()
  ]}
]
variables [Parachute open {type=boolean}]
